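; Blueprint for an environmental lock scanner's screen, showing scanning, success, failure and target-lost states through a UI_ScannerScreen widget component.
; Each state event unhides the screen, enables its tick and forwards to the widget, with text for result states.
; ErrorTimeline blends the screen material's inner and outer glow from blue to red on failure, reversing for other states.
; Results start a dismiss timer that hides the screen and stops its tick; scanning clears that timer.
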(event EventBeginPlay
  (bind _asui_scanner_screen (Utilities|Casting|CastToUI_ScannerScreen (UserInterface|GetUserWidgetObject (Variables|Default|GetScreenWidget))))
  (Variables|Default|SetWidget_Screen _asui_scanner_screen)
  (Components|Tick|SetComponentTickEnabled (Variables|Default|GetScreenWidget) false))

(event Custom|ScreenEventScan
  (CallFunction|ScreenEventError true)
  (Class|UIScannerScreen|EventScanning (Variables|Default|GetWidget_Screen))
  (Utilities|Time|ClearandInvalidateTimerbyHandle (Variables|Default|GetTimeoutHandle)))

(event Custom|ScreenEventFailure
  (params (FailureText :type Text))
  (CallFunction|ScreenEventError false)
  (Class|UIScannerScreen|EventFail (Variables|Default|GetWidget_Screen) FailureText)
  (bind _returnvalue (Utilities|Time|SetTimerbyEvent (EventDispatchers|CreateEvent :!event Custom|ScreenEventDismiss) (Variables|Default|GetScreenEventDismissTimerDuration)))
  (Variables|Default|SetTimeoutHandle _returnvalue))

(event Custom|ScreenEventSuccess
  (params (SuccessText :type Text))
  (CallFunction|ScreenEventError true)
  (Class|UIScannerScreen|EventSuccess (Variables|Default|GetWidget_Screen) SuccessText)
  (bind _returnvalue (Utilities|Time|SetTimerbyEvent (EventDispatchers|CreateEvent :!event Custom|ScreenEventDismiss) (Variables|Default|GetScreenEventDismissTimerDuration)))
  (Variables|Default|SetTimeoutHandle _returnvalue))

(event Custom|ScreenEventTargetLost
  (params (LostText :type Text))
  (CallFunction|ScreenEventError true)
  (Class|UIScannerScreen|EventTargetLost (Variables|Default|GetWidget_Screen) LostText)
  (bind _returnvalue (Utilities|Time|SetTimerbyEvent (EventDispatchers|CreateEvent :!event Custom|ScreenEventDismiss) (Variables|Default|GetScreenEventDismissTimerDuration)))
  (Variables|Default|SetTimeoutHandle _returnvalue))

(event Custom|ScreenEventDismiss
  (Development|SetHiddeninGame (Variables|Default|GetScreenWidget) true false)
  (Components|Tick|SetComponentTickEnabled (Variables|Default|GetScreenWidget) false))

(event Custom|ScreenEventError
  (params (Reverse :type Boolean))
  (bind _screenwidget (Variables|Default|GetScreenWidget))
  (bind _returnvalue (UserInterface|GetMaterialInstance _screenwidget))
  (bind _returnvalue_2 (Math|Color|MakeColor 0.0 2.0 15.0))
  (bind _returnvalue_3 (Math|Color|MakeColor 4e+01 1.2 3.0))
  (Development|SetHiddeninGame (Variables|Default|GetScreenWidget) false false)
  (Components|Tick|SetComponentTickEnabled (Variables|Default|GetScreenWidget) true)
  (if Reverse
    (bind (_direction _switch) (Timeline|ErrorTimeline Reverse)
      (:Update
        (Rendering|Material|SetVectorParameterValue _returnvalue "Glow Color Inner" (Math|Color|Lerp(LinearColor) _returnvalue_2 _returnvalue_3 _switch) :!class /Script/Engine.MaterialInstanceDynamic)
        (Rendering|Material|SetVectorParameterValue _returnvalue "Glow Color Outer" (Math|Color|LinearColor*Float (Math|Color|Lerp(LinearColor) _returnvalue_2 _returnvalue_3 _switch) 2.0) :!class /Script/Engine.MaterialInstanceDynamic)))
    (else
      (Timeline|ErrorTimeline Play))))

(fn ConstructionScript)
