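# Material for the lock scanner's eye, a glossy lit surface with a glowing iris that blends from locked to unlocked color.
# Local position, offset by a look direction, forms radial sine rings or a pulsing solid glow.
# Custom primitive data drives color, brightness and gaze, blending curve-atlas wandering, staring at the player and looking straight ahead.
# Fresnel over a noise normal darkens the base color for a glass look.
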Material /Game/Athena/Items/EnvironmentalItems/Locks/Scanner/Materials/M_Lock_Scanner_Eye
graph {
 properties {pixelDepthOffsetMode=PDOM_Legacy, usedShadingModels=Default Lit}
 nodes [
  n1 = VectorParameter() {bUseCustomPrimitiveData=true, parameterName=Base Color Locked, primitiveDataIndex=18, defaultValue=<rgba 1 0 0 0>}
  n2 = VectorParameter() {bUseCustomPrimitiveData=true, parameterName=Base Color Unlocked, primitiveDataIndex=22, defaultValue=<rgba 0 1 0 0>}
  n3 = LinearInterpolate(A=@n1.RGB, B=@n2.RGB, Alpha=ScalarParameter() {bUseCustomPrimitiveData=true, parameterName=LockUnlockState, primitiveDataIndex=1})
  n4 = ComponentMask(None=Divide(A=PreSkinnedPosition(), B=ObjectBounds())) {b=true, g=true}
  n5 = ScalarParameter() {defaultValue=0.2, parameterName=WaterTiling}
  n6 = TextureSampleParameter2D(UVs=VertexInterpolator(VS=Multiply(A=@n4, B=@n5)).PS) {
   parameterName=Noise Normal
   samplerSource=SSM_Wrap_WorldGroupSettings
   samplerType=SAMPLERTYPE_Normal
   texture=/Game/Global/Textures/Default/Noise/T_Noise_Normal_01
  }
  n7 = Normalize(VectorInput=Multiply(A=@n6.RGB, B=VectorParameter() {parameterName=Normal Intensity, defaultValue=<rgba 0.4 0.4 0.4 0>}.RGB))
  n8 = Multiply(
   A=Fresnel(ExponentIn=ScalarParameter() {defaultValue=1, parameterName=Fresnel Exponent}, Normal=Transform(None=@n7))
   B=ScalarParameter() {defaultValue=1, parameterName=Glass thickness}
  )
  LinearInterpolate(
   A=Multiply(A=@n3, B=ScalarParameter() {defaultValue=0.035, parameterName=Base Color Brightness})
   Alpha=LinearInterpolate(
    A=ScalarParameter() {parameterName=Fresnel Inner Transparency}
    B=ScalarParameter() {defaultValue=0.85, parameterName=Fresnel Outer Transparency}
    Alpha=Clamp(None=@n8)
   )
  ) {constB=0}
  n9 = MaterialFunctionCall(Float3=Constant3Vector() {constant=<rgba 25 25 25 0>}) {materialFunction=&BreakOutFloat3Components}
  n10 = MaterialFunctionCall(R=@n9.R, G=@n9.R, B=Multiply(A=@n9.B)) {materialFunction=&AppendMany}
  n11 = Constant() {r=0.55}
  n12 = Add(
   A=Divide(
    A=MaterialFunctionCall() {materialFunction=&LocalPosition}.Local Position
    B=Multiply(A=@n10.RGB, B=@n11)
   )
   B=VectorParameter() {parameterName=Offset, defaultValue=<rgba 0 0 -0.4 1>}.RGB
  )
  n13 = VectorParameter() {
   bUseCustomPrimitiveData=true
   parameterName=ObjectLocation
   primitiveDataIndex=14
   channelNames {a="", b=Z, g=Y, r=X}
   defaultValue=<rgba 0 0 0 1>
  }
  n14 = Subtract(
   A=ObjectPositionWS()
   B=MaterialFunctionCall(R=@n13.X, G=@n13.X, B=@n13.X) {materialFunction=&AppendMany}.RGB
  )
  n15 = MaterialFunctionCall(Float3=WorldPosition().XYZ) {materialFunction=&BreakOutFloat3Components}
  n16 = Add(A=@n15.R, B=@n15.R)
  n17 = Add(A=Multiply(A=Add(A=@n16, B=@n15.B)) {constB=0.0025}, B=Time() {bIgnorePause=true})
  n18 = LinearInterpolate(
   A=Multiply(A=@n17) {constB=0.125}
   B=@n17
   Alpha=Power(Base=ScalarParameter() {bUseCustomPrimitiveData=true, defaultValue=1, parameterName=StareIntently, primitiveDataIndex=13})
  )
  n19 = CurveAtlasRowParameter(CurveTime=Frac(None=@n18)) {
   atlas=&AthenaCurveAtlas01
   curve=/Game/Characters/Enemies/Cattus_Head/Materials/Cattus_EyeMovement_Curve
   defaultValue=92
  }
  n20 = Cosine(None=@n18)
  n21 = Sine(None=@n18)
  n22 = Add(
   A=Normalize(VectorInput=Transform(None=@n14) {transformSourceType=TRANSFORMSOURCE_World, transformType=TRANSFORM_Local})
   B=MaterialFunctionCall(R=Constant() {r=2}, G=Multiply(A=@n19.G, B=@n20), B=Multiply(A=@n19.B, B=@n21)) {materialFunction=&AppendMany}.RGB
  )
  n23 = VectorParameter() {
   bUseCustomPrimitiveData=true
   parameterName=PlayerLocation
   primitiveDataIndex=8
   channelNames {a="", b=Z, g=Y, r=X}
   defaultValue=<rgba 0 0 0 1>
  }
  n24 = Subtract(
   A=ObjectPositionWS()
   B=MaterialFunctionCall(R=@n23.X, G=@n23.X, B=@n23.X) {materialFunction=&AppendMany}.RGB
  )
  n25 = Multiply(A=Normalize(VectorInput=Transform(None=@n24) {transformSourceType=TRANSFORMSOURCE_World, transformType=TRANSFORM_Local})) {constB=0.75}
  n26 = Multiply(A=@n22) {constB=0.025}
  n27 = VectorParameter() {
   bUseCustomPrimitiveData=true
   parameterName=ObjectLocation
   primitiveDataIndex=14
   channelNames {a="", b=Z, g=Y, r=X}
   defaultValue=<rgba 0 0 0 1>
  }
  n28 = Subtract(A=ObjectPositionWS(), B=@n27)
  n29 = LinearInterpolate(
   A=LinearInterpolate(
    A=Multiply(A=@n22) {constB=0.25}
    B=Add(A=@n25, B=@n26)
    Alpha=ScalarParameter() {bUseCustomPrimitiveData=true, defaultValue=1, parameterName=Stare, primitiveDataIndex=12}
   )
   B=Normalize(VectorInput=Transform(None=@n28) {transformSourceType=TRANSFORMSOURCE_World, transformType=TRANSFORM_Local})
   Alpha=ScalarParameter() {bUseCustomPrimitiveData=true, defaultValue=1, parameterName=LookStraightAhead, primitiveDataIndex=2}
  )
  n30 = MaterialFunctionCall(Vector or UVs=ComponentMask(None=Add(A=@n12, B=@n29)) {b=true, g=true}) {materialFunction=&VectorToRadialValue}
  n31 = OneMinus(None=MaterialFunctionCall(UV Channel=@n30.Radial Coordinates) {materialFunction=&LinearGradient}.VGradient)
  n32 = Sine(None=Time())
  n33 = Multiply(
   A=Power(Base=@n31, Exp=ScalarParameter() {defaultValue=7, parameterName=No Rings Fade Power})
   B=LinearInterpolate(Alpha=ConstantBiasScale(None=@n32)) {constA=0.45}
  )
  n34 = ScalarParameter() {bUseCustomPrimitiveData=true, parameterName=SwapToSolid, primitiveDataIndex=3}
  n35 = Constant() {r=1}
  n36 = Constant() {r=1}
  n37 = LinearInterpolate(
   A=ScalarParameter() {bUseCustomPrimitiveData=true, parameterName=RingCountHigh, primitiveDataIndex=26}
   B=ScalarParameter() {bUseCustomPrimitiveData=true, parameterName=RingCountLow, primitiveDataIndex=27}
   Alpha=@n36
  ) {constA=12, constB=4}
  n38 = MaterialFunctionCall(UV Channel=Multiply(
    A=@n30.Radial Coordinates
    B=MaterialFunctionCall(R=@n35, G=@n37) {materialFunction=&AppendMany}.RG
   )) {materialFunction=&LinearGradient}
  n39 = Multiply(A=Multiply(A=Sine(None=@n38.VGradient)) {constB=0.75}, B=Power(Base=@n31, Exp=ScalarParameter() {defaultValue=4, parameterName=Fade Power}))
  n40 = Clamp(None=@n3)
  n41 = Multiply(A=@n40) {constB=3}
  n42 = Multiply(A=@n40) {constB=10}
  Multiply(
   A=Multiply(A=LinearInterpolate(A=@n39, B=@n33, Alpha=@n34), B=LinearInterpolate(A=@n41, B=@n42, Alpha=@n36))
   B=Add(A=Constant() {r=1}, B=ScalarParameter() {bUseCustomPrimitiveData=true, parameterName=Emissive Multiplier, primitiveDataIndex=7})
  )
  ScalarParameter() {defaultValue=0.1, parameterName=Metallic}
  ScalarParameter() {defaultValue=1, parameterName=Specular}
  ScalarParameter() {defaultValue=0.1, parameterName=Roughness}
 ]
}
