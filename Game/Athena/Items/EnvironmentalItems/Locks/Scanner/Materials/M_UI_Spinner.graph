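# Translucent UI material for a scanner lock spinner, drawing a segmented, rotating ring that fades on a repeating cycle.
# Time times Speed drives a Rotator and a looping phase that feeds the ring radius, segment threshold and fade.
# Subtracting two SmoothSteps on a radial gradient forms the ring.
# A gradient texture sampled by angle, repeated by floored Segments and soft-thresholded, cuts it into segments.
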Material /Game/Athena/Items/EnvironmentalItems/Locks/Scanner/Materials/M_UI_Spinner
graph {
 properties {blendMode=BLEND_Translucent, materialDomain=MD_UI, pixelDepthOffsetMode=PDOM_Legacy, usedShadingModels=Default Lit}
 nodes [
  Constant() {r=1}
  LinearInterpolate()
  n1 = Multiply(A=Time(), B=ScalarParameter() {defaultValue=1, parameterName=Speed})
  n2 = Abs(None=@n1)
  MaterialFunctionCall(Value=@n2) {materialFunction=&LinearSine}
  n3 = MaterialFunctionCall() {materialFunction=&RadialGradientExponential}
  n4 = VectorParameter() {parameterName=Scale, defaultValue=<rgba 2 0 0 1>}
  n5 = Fmod(A=@n2, B=Constant() {r=1})
  n6 = LinearInterpolate(A=@n4.R, B=@n4.R, Alpha=@n5) {constA=1, constB=0}
  n7 = MaterialFunctionCall(Alpha=@n3.RadialGradientExponential, Min=@n6, Max=Add(A=@n6, B=ScalarParameter() {defaultValue=0.002, parameterName=EdgeSmooth})) {materialFunction=&SmoothStep}
  n8 = Add(A=@n6, B=ScalarParameter() {defaultValue=0.1, parameterName=RingSize})
  n9 = MaterialFunctionCall(Alpha=@n3.RadialGradientExponential, Min=@n8, Max=Add(A=@n8, B=ScalarParameter() {defaultValue=0.002, parameterName=EdgeSmooth})) {materialFunction=&SmoothStep}
  n10 = Add(A=Rotator(Coordinate=TextureCoordinate(), Time=@n1), B=Constant2Vector() {g=-0.5, r=-0.5})
  n11 = ScalarParameter() {defaultValue=10, parameterName=Segments}
  n12 = TextureSample(UVs=Multiply(
    A=MaterialFunctionCall(Vector or UVs=@n10) {materialFunction=&VectorToRadialValue}.Vector Converted to Angle
    B=Floor(None=@n11)
   )) {texture=/Game/Global/Textures/Default/Masks_and_Gradients/T_UI_ReticleSegmentGradient}
  n13 = VectorParameter() {parameterName=SegmentRange, defaultValue=<rgba 0 1 0 1>}
  n14 = Power(Base=@n5)
  n15 = LinearInterpolate(A=@n13.R, B=@n13.R, Alpha=@n14)
  n16 = Add(A=@n15, B=ScalarParameter() {defaultValue=0.02, parameterName=EdgeSoftness})
  Multiply(
   A=Multiply(
    A=Subtract(A=@n7.Result, B=@n9.Result)
    B=MaterialFunctionCall(Alpha=@n12.R, Min=@n15, Max=@n16) {materialFunction=&SmoothStep}.Result
   )
   B=OneMinus(None=@n14)
  )
  ScalarParameter() {defaultValue=0.8, parameterName=Edge}
  ScalarParameter() {defaultValue=0.005, parameterName=OuterSize}
  ScalarParameter() {defaultValue=0.005, parameterName=OuterSize}
 ]
}
