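# Screen widget for the scanner lock that shows Scanning, Access Granted, Access Denied and Target Lost states.
# Four custom events switch a widget switcher between states; the three outcome events accept custom text, with default wording when empty.
# Scanning and Target Lost show the spinner-material background, while Success and Fail hide it.
# Target Lost also plays FlashText, which blinks the text border's opacity.
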WidgetBlueprint /Game/Athena/Items/EnvironmentalItems/Locks/Scanner/UI_ScannerScreen
animations [FlashText {
  binding_count=1
  length_seconds=0.7758
  object=&self:FlashText
  sequence {bindings [CommonBorder_40 {
     id=#1
     possessed_class=CommonBorder
     tracks [{
       type=MovieSceneFloatTrack
       display_name=Render Opacity
       sections [{
         type=MovieSceneFloatSection
         channels [None {type=MovieSceneScriptingFloatChannel, keys [[0,1],[2.25,0],[4.5,1],[6.75,0],[9,1],[11.25,0],[13.5,1],[15.516,1]]}]
        }]
      }]
    }]}
 }]
graphs [EventGraph {nodes [
   n1 = EventTargetLost!() -> (OutputDelegate, then, ScreenText:Text) {is=title K2Node_CustomEvent}
   > "SetText (Text)"(
    self=Get Text_Misc() -> (Text_Misc:Common Text Object Reference)
    InText=Select(Option 0:Text=@n1.ScreenText, Option 1:Text="LOCTEXT(\"Target\\r\\nLost\")", Index=TextIsEmpty(InText=@n1.ScreenText)) -> (ReturnValue:Text)
   )
   n2 = Get CommonWidgetSwitcher_States() -> (CommonWidgetSwitcher_States:Common Widget Switcher Legacy Object Reference)
   > SetActiveWidgetIndex(self=@n2, Index=3)
   > n3 = PlayAnimation(InAnimation=Get FlashText() -> (FlashText:Widget Animation Object Reference))
   n4 = Get Overlay_BGIdle() -> (Overlay_BGIdle:Overlay Object Reference)
   SetVisibility(execute=[@n7.then, @n3.then], self=@n4, InVisibility=Visible)
   n5 = EventFail!() -> (OutputDelegate, then, ScreenText:Text) {is=title K2Node_CustomEvent}
   > "SetText (Text)"(
    self=Get Text_Failure() -> (Text_Failure:Common Text Object Reference)
    InText=Select(Option 0=@n5.ScreenText, Option 1="LOCTEXT(\"Access\\r\\nDenied\")", Index=TextIsEmpty(InText=@n5.ScreenText))
   )
   > n6 = SetActiveWidgetIndex(self=@n2, Index=2)
   SetVisibility(execute=[@n6.then, @n9.then], self=@n4, InVisibility=Hidden)
   EventScanning() {is=title K2Node_CustomEvent}
   > n7 = SetActiveWidgetIndex(self=@n2, Index=0)
   n8 = EventSuccess!() -> (OutputDelegate, then, ScreenText:Text) {is=title K2Node_CustomEvent}
   > "SetText (Text)"(
    self=Get Text_Success() -> (Text_Success:Common Text Object Reference)
    InText=Select(Option 0=@n8.ScreenText, Option 1="LOCTEXT(\"Access\\r\\nGranted\")", Index=TextIsEmpty(InText=@n8.ScreenText))
   )
   > n9 = SetActiveWidgetIndex(self=@n2, Index=1)
  ]}]
variables []
widgets {
 info {inheritedWidgetCount=0, parentClass=&UserWidget, rootWidgetClass=&Overlay, widgetCount=17}
 widgets [
  Overlay_0 {widgetClassPath=&Overlay, widget=Overlay_0}
  CommonBorder_42 {
   widgetClassPath=&CommonBorder
   parent=Overlay_0
   widget=CommonBorder_42
   properties {padding=<ltrb 0 0 0 0>}
   slot {class=OverlaySlot, properties {horizontalAlignment=HAlign_Center, verticalAlignment=VAlign_Center}}
  }
  SizeBox_0 {
   widgetClassPath=&SizeBox
   parent=CommonBorder_42
   widget=SizeBox_0
   properties {bOverride_HeightOverride=true, bOverride_WidthOverride=true, heightOverride=120, widthOverride=200}
   slot {class=BorderSlot, properties {padding=<ltrb 0 0 0 0>}}
  }
  ScaleBox_0 {widgetClassPath=&ScaleBox, parent=SizeBox_0, widget=ScaleBox_0, slot {class=SizeBoxSlot}}
  Overlay_1 {
   widgetClassPath=&Overlay
   parent=ScaleBox_0
   widget=Overlay_1
   slot {class=ScaleBoxSlot, properties {horizontalAlignment=HAlign_Fill, verticalAlignment=VAlign_Fill}}
  }
  Overlay_BGIdle {
   widgetClassPath=&Overlay
   bIsVariable=true
   parent=Overlay_1
   widget=Overlay_BGIdle
   slot {class=OverlaySlot, properties {horizontalAlignment=HAlign_Fill, verticalAlignment=VAlign_Fill, padding=<ltrb -200 -200 -200 -200>}}
  }
  Image_100 {
   widgetClassPath=&Image
   bIsVariable=true
   parent=Overlay_BGIdle
   widget=Image_100
   properties {brush {
     bIsDynamicallyLoaded=false
     drawAs=Image
     imageType=NoImage
     mirroring=NoMirror
     resourceName=None
     resourceObject=/Game/Athena/Items/EnvironmentalItems/Locks/Scanner/Materials/M_UI_Spinner_Inst
     tiling=NoTile
     imageSize=<100 100>
     margin=<ltrb 0 0 0 0>
     outlineSettings {
      bUseBrushTransparency=false
      roundingType=HalfHeightRadius
      width=0
      color {colorUseRule=UseColor_Specified, specifiedColor=<rgba 0 0 0 0>}
      cornerRadii=<0 0 0 0>
     }
     tintColor {colorUseRule=UseColor_Specified, specifiedColor=<rgba 1 1 1 1>}
     uVRegion {bIsValid=false, max=<0 0>, min=<0 0>}
    }}
   slot {class=OverlaySlot, properties {horizontalAlignment=HAlign_Center, verticalAlignment=VAlign_Center}}
  }
  Image {
   widgetClassPath=&Image
   bIsVariable=true
   parent=Overlay_BGIdle
   widget=Image
   properties {brush {
     bIsDynamicallyLoaded=false
     drawAs=Image
     imageType=NoImage
     mirroring=NoMirror
     resourceName=None
     resourceObject=/Game/Athena/Items/EnvironmentalItems/Locks/Scanner/Materials/M_UI_Spinner_Inst1
     tiling=NoTile
     imageSize=<100 100>
     margin=<ltrb 0 0 0 0>
     outlineSettings {
      bUseBrushTransparency=false
      roundingType=HalfHeightRadius
      width=0
      color {colorUseRule=UseColor_Specified, specifiedColor=<rgba 0 0 0 0>}
      cornerRadii=<0 0 0 0>
     }
     tintColor {colorUseRule=UseColor_Specified, specifiedColor=<rgba 1 1 1 1>}
     uVRegion {bIsValid=false, max=<0 0>, min=<0 0>}
    }}
   slot {class=OverlaySlot, properties {horizontalAlignment=HAlign_Center, verticalAlignment=VAlign_Center, padding=<ltrb -200 -200 -200 -200>}}
  }
  Image_ScreenBorder {
   widgetClassPath=&Image
   bIsVariable=true
   parent=Overlay_1
   widget=Image_ScreenBorder
   properties {brush {
     bIsDynamicallyLoaded=false
     drawAs=Box
     imageType=NoImage
     mirroring=NoMirror
     resourceName=None
     resourceObject=/Game/Athena/HUD/Vehicles/Materials/M_UI_TargetBracket_Inst
     tiling=NoTile
     imageSize=<64 64>
     margin=<ltrb 0.4 0.4 0.4 0.4>
     outlineSettings {
      bUseBrushTransparency=false
      roundingType=HalfHeightRadius
      width=0
      color {colorUseRule=UseColor_Specified, specifiedColor=<rgba 0 0 0 0>}
      cornerRadii=<0 0 0 0>
     }
     tintColor {colorUseRule=UseColor_Specified, specifiedColor=<rgba 1 1 1 1>}
     uVRegion {bIsValid=false, max=<0 0>, min=<0 0>}
    }}
   slot {class=OverlaySlot, properties {horizontalAlignment=HAlign_Fill, verticalAlignment=VAlign_Fill, padding=<ltrb 10 10 10 10>}}
  }
  CommonBorder_40 {
   widgetClassPath=&CommonBorder
   parent=Overlay_1
   widget=CommonBorder_40
   properties {padding=<ltrb 0 0 0 0>}
   slot {class=OverlaySlot, properties {horizontalAlignment=HAlign_Fill, verticalAlignment=VAlign_Fill, padding=<ltrb 20 20 20 20>}}
  }
  ScaleBox_1 {
   widgetClassPath=&ScaleBox
   parent=CommonBorder_40
   widget=ScaleBox_1
   slot {class=BorderSlot, properties {padding=<ltrb 0 0 0 0>}}
  }
  CommonWidgetSwitcher_States {
   widgetClassPath=&CommonWidgetSwitcherLegacy
   bIsVariable=true
   parent=ScaleBox_1
   widget=CommonWidgetSwitcher_States
   properties {transitionDuration=0, transitionType=Zoom}
   slot {class=ScaleBoxSlot}
  }
  Overlay_2 {widgetClassPath=&Overlay, parent=CommonWidgetSwitcher_States, widget=Overlay_2, slot {class=WidgetSwitcherSlot}}
  CommonTextBlock_1 {
   widgetClassPath=&CommonTextBlock
   parent=Overlay_2
   widget=CommonTextBlock_1
   properties {
    justification=Center
    text=Scanning
    textTransformPolicy=ToUpper
    font {
     bForceMonospaced=false
     bMaterialIsStencil=false
     fontMaterial=null
     fontObject=&Burbank
     letterSpacing=0
     monospacedWidth=1
     size=18
     skewAmount=0
     typefaceFontName=BurbankRegular-Black
     outlineSettings {
      bApplyOutlineToDropShadows=false
      bMiteredCorners=false
      bSeparateFillAlpha=false
      outlineMaterial=null
      outlineSize=0
      outlineColor=<rgba 0 0 0 1>
     }
    }
    shadowOffset=<0 0>
   }
   slot {class=OverlaySlot, properties {verticalAlignment=VAlign_Center}}
  }
  Text_Success {
   widgetClassPath=&CommonTextBlock
   bIsVariable=true
   parent=CommonWidgetSwitcher_States
   widget=Text_Success
   properties {
    justification=Center
    lineHeightPercentage=0.8
    text="Access\r\nGranted"
    textTransformPolicy=ToUpper
    font {
     bForceMonospaced=false
     bMaterialIsStencil=false
     fontMaterial=null
     fontObject=&Burbank
     letterSpacing=0
     monospacedWidth=1
     size=18
     skewAmount=0
     typefaceFontName=BurbankRegular-Black
     outlineSettings {
      bApplyOutlineToDropShadows=false
      bMiteredCorners=false
      bSeparateFillAlpha=false
      outlineMaterial=null
      outlineSize=0
      outlineColor=<rgba 0 0 0 1>
     }
    }
    shadowOffset=<0 0>
   }
   slot {class=WidgetSwitcherSlot, properties {horizontalAlignment=HAlign_Center, verticalAlignment=VAlign_Center}}
  }
  Text_Failure {
   widgetClassPath=&CommonTextBlock
   bIsVariable=true
   parent=CommonWidgetSwitcher_States
   widget=Text_Failure
   properties {
    justification=Center
    lineHeightPercentage=0.8
    text="Access\r\nDenied"
    textTransformPolicy=ToUpper
    font {
     bForceMonospaced=false
     bMaterialIsStencil=false
     fontMaterial=null
     fontObject=&Burbank
     letterSpacing=0
     monospacedWidth=1
     size=18
     skewAmount=0
     typefaceFontName=BurbankRegular-Black
     outlineSettings {
      bApplyOutlineToDropShadows=false
      bMiteredCorners=false
      bSeparateFillAlpha=false
      outlineMaterial=null
      outlineSize=0
      outlineColor=<rgba 0 0 0 1>
     }
    }
    shadowOffset=<0 0>
   }
   slot {class=WidgetSwitcherSlot, properties {horizontalAlignment=HAlign_Center, verticalAlignment=VAlign_Center}}
  }
  Text_Misc {
   widgetClassPath=&CommonTextBlock
   bIsVariable=true
   parent=CommonWidgetSwitcher_States
   widget=Text_Misc
   properties {
    justification=Center
    lineHeightPercentage=0.8
    text="Target\r\nLost"
    textTransformPolicy=ToUpper
    font {
     bForceMonospaced=false
     bMaterialIsStencil=false
     fontMaterial=null
     fontObject=&Burbank
     letterSpacing=0
     monospacedWidth=1
     size=18
     skewAmount=0
     typefaceFontName=BurbankRegular-Black
     outlineSettings {
      bApplyOutlineToDropShadows=false
      bMiteredCorners=false
      bSeparateFillAlpha=false
      outlineMaterial=null
      outlineSize=0
      outlineColor=<rgba 0 0 0 1>
     }
    }
    shadowOffset=<0 0>
   }
   slot {class=WidgetSwitcherSlot, properties {horizontalAlignment=HAlign_Center, verticalAlignment=VAlign_Center}}
  }
 ]
}
